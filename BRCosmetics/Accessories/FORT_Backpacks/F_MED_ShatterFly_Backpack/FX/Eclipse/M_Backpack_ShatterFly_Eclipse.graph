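# M_Backpack_ShatterFly_Eclipse is an unlit, two-sided FX material for the Eclipse effects of the ShatterFly backpack cosmetic.
# Its small graph feeds a constant of 1 into the Emissive input of the shared MF_FX_EmissiveAdjustments function, with no textures or other nodes.
# Usage flags enable it on skeletal meshes, Niagara mesh particles and mesh deformers.
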Material /BRCosmetics/Accessories/FORT_Backpacks/F_MED_ShatterFly_Backpack/FX/Eclipse/M_Backpack_ShatterFly_Eclipse
graph {
 properties {
  bUsedWithMeshDeformer=true
  bUsedWithNiagaraMeshParticles=true
  bUsedWithSkeletalMesh=true
  pixelDepthOffsetMode=PDOM_Legacy
  shadingModel=MSM_Unlit
  twoSided=true
  usedShadingModels=Unlit
 }
 nodes [MaterialFunctionCall(Emissive=Constant() {r=1}) {materialFunction=&MF_FX_EmissiveAdjustments}]
}
